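# Material function for the 3DHB TextureBaker tool that Gaussian-blurs an input texture, with DeNoise and Power adjustments.
# Samples is scaled, clamped and offset to size a weighted 2D tap grid, while BlurScale sets tap spacing.
# The custom HLSL loop subtracts DeNoise and applies Power per tap before normalising.
# A dummy Custom node injects the gaussian helper by closing its own function early; a sampletest node is unused.
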MaterialFunction /Game/Tools/3DHB/TextureBaker/BakeShader/MF_3DHB_GaussianBlur
graph {nodes [
  MaterialExpressionFunctionInput_0 = FunctionInput() {inputName=Texture, inputType=FunctionInput_Texture2D}
  MaterialExpressionTextureCoordinate_1 = TextureCoordinate()
  Custom!(Texture=@MaterialExpressionFunctionInput_0, UV=@MaterialExpressionTextureCoordinate_1) {
   code="return Texture2DSampleLevel(Texture,TextureSampler,UV,0);"
   description=sampletest
   outputType=CMOT_Float4
   inputs [
    {inputName=Texture, input {expression=&self:MaterialExpressionFunctionInput_0}}
    {inputName=UV, input {expression=&self:MaterialExpressionTextureCoordinate_1}}
   ]
  }
  MaterialExpressionCustom_0 = Custom!() {
   code="return 0;\r\n}\r\nhalf gaussian(half2 i , half sigma){\r\n    return exp(-0.5*dot(i/=sigma,i)) / (6.28*sigma*sigma);\r\n\r\n"
   description=Gaussian
   outputType=CMOT_Float1
   inputs []
  }
  n1 = Multiply(A=FunctionInput() {inputName=Samples, inputType=FunctionInput_Scalar}) {constB=64}
  MaterialExpressionAdd_0 = Add(A=Min(A=Max(A=@n1) {constB=0}) {constB=512}) {constB=2}
  MaterialExpressionDivide_0 = Divide(A=FunctionInput() {inputName=BlurScale, inputType=FunctionInput_Vector2}) {constB=1000}
  MaterialExpressionTextureCoordinate_0 = TextureCoordinate()
  MaterialExpressionFunctionInput_3 = FunctionInput() {inputName=Power, inputType=FunctionInput_Scalar}
  MaterialExpressionFunctionInput_4 = FunctionInput() {inputName=DeNoise, inputType=FunctionInput_Scalar}
  n2 = Custom!(
   GaussianF=@MaterialExpressionCustom_0
   samples=@MaterialExpressionAdd_0
   scale=@MaterialExpressionDivide_0
   UV=@MaterialExpressionTextureCoordinate_0
   Texture=@MaterialExpressionFunctionInput_0
   Power=@MaterialExpressionFunctionInput_3
   DeNoise=@MaterialExpressionFunctionInput_4
  ) {
   code="const float sigma = float(samples) * .25;\r\nhalf2 U = UV;\r\n\r\nhalf4 O = 0;  \r\nint s = samples/2;\r\nhalf W = 0;\r\nfor ( int i = 0; i < s*s; i++ ) {\r\n    half2 d = half2(i%s, i/s)*2.0 - half(samples)/2.0;\r\n    half g = gaussian(d,sigma);\r\n    half4 sampledcolor = Texture2DSampleLevel(Texture,TextureSampler,U + scale * d,0);\r\n    half4 adjustedcolor = half4(pow(clamp(sampledcolor.xyz-DeNoise,0.0,1.0),Power),sampledcolor.w);\r\n    O += g * adjustedcolor;\r\n    W += g;\r\n}\r\nreturn O/W;"
   description=Blur
   outputType=CMOT_Float4
   inputs [
    {inputName=GaussianF, input {expression=&self:MaterialExpressionCustom_0}}
    {inputName=samples, input {expression=&self:MaterialExpressionAdd_0}}
    {inputName=scale, input {expression=&self:MaterialExpressionDivide_0}}
    {inputName=UV, input {expression=&self:MaterialExpressionTextureCoordinate_0}}
    {inputName=Texture, input {expression=&self:MaterialExpressionFunctionInput_0}}
    {inputName=Power, input {expression=&self:MaterialExpressionFunctionInput_3}}
    {inputName=DeNoise, input {expression=&self:MaterialExpressionFunctionInput_4}}
   ]
  }
  FunctionOutput(None=@n2)
 ]}
